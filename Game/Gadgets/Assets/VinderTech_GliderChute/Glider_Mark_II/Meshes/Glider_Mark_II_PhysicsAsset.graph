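# Physics asset for the VinderTech Glider Mark II glider chute, giving its skeletal mesh one capsule body per listed bone.
# Constraints share 45-degree swing and twist limit values, chaining flaps, hinges and crossbars to each wing base and four canopy pieces to the parachute.
# Only the right engine and wing bases link to a tiny Body_Offset_r body.
# The left bases, both handles and the parachute have no parent constraint.
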PhysicsAsset /Game/Gadgets/Assets/VinderTech_GliderChute/Glider_Mark_II/Meshes/Glider_Mark_II_PhysicsAsset
bodies {
 Body_Offset_r [{shapeType=Capsule, length=0.505, radius=0.505, rotation=<rot 0 -90 0>}]
 Crossbar_l [{shapeType=Capsule, length=22.5891, radius=6.4637, center=<-6.0346 -0.2977 17.575>, rotation=<rot -0.6909 -90 -19.718>}]
 Crossbar_r [{shapeType=Capsule, length=25.0302, radius=6.5567, center=<6.8675 0.3422 -19.8213>, rotation=<rot -0.6122 -90 -20.0792>}]
 Engine_Base_l [{shapeType=Capsule, length=54.5549, radius=22.4023, center=<-9.0966 -50.0649 -5.7679>, rotation=<rot 0 -2.806 -86.4092>}]
 Engine_Base_r [{shapeType=Capsule, length=54.5541, radius=22.4036, center=<9.0961 50.0618 5.7626>, rotation=<rot 0 -2.8079 -86.4193>}]
 Front_Flap_l [{shapeType=Capsule, length=41.8672, radius=8.908, center=<-6.2673 11.6402 -1.5745>, rotation=<rot 0 -156.1282 -89.8437>}]
 Front_Flap_r [{shapeType=Capsule, length=41.8605, radius=8.9107, center=<6.2901 -11.6309 1.4641>, rotation=<rot 0 23.8342 -90>}]
 Handle_Base_l [{shapeType=Capsule, length=33.6073, radius=9.6986, center=<23.1681 9.3127 -9.2749>, rotation=<rot 0 -65.7795 -77.1561>}]
 Handle_Base_r [{shapeType=Capsule, length=33.6073, radius=9.6986, center=<-23.1681 9.3127 -9.275>, rotation=<rot 0 65.7795 -77.1561>}]
 Parachute [{shapeType=Capsule, length=64.0414, radius=63.947, center=<0 1.8513 -16.5954>, rotation=<rot 0 0 -88.9069>}]
 Parachute_Back_l [{shapeType=Capsule, length=32.6527, radius=27.486, center=<-26.0072 0.4215 4.2709>, rotation=<rot 0.6203 166.4457 -90>}]
 Parachute_Back_r [{shapeType=Capsule, length=32.6527, radius=27.486, center=<26.007 -0.4217 -4.2712>, rotation=<rot 0.6203 166.4457 -90>}]
 Parachute_Front_l [{shapeType=Capsule, length=26.9787, radius=22.9671, center=<-19.2237 -3.0576 3.3486>, rotation=<rot 6.542 -174.5826 -90>}]
 Parachute_Front_r [{shapeType=Capsule, length=26.9787, radius=22.9671, center=<19.2236 3.0576 -3.3485>, rotation=<rot 6.542 -174.5826 -90>}]
 Wing_Base_l [{shapeType=Capsule, length=86.7041, radius=40.4811, center=<-65.9042 -27.5221 -4.7176>, rotation=<rot 0 -58.1397 -86.5076>}]
 Wing_Base_r [{shapeType=Capsule, length=86.7012, radius=40.4892, center=<65.9049 27.5159 4.7073>, rotation=<rot 0 -58.1291 -86.4996>}]
 Wing_Flap_Hinge_l [{shapeType=Capsule, length=25.9853, radius=7.765, center=<-3.8051 6.8724 0.2407>, rotation=<rot 0 89.0417 -89.6665>}]
 Wing_Flap_Hinge_r [{shapeType=Capsule, length=25.9853, radius=7.765, center=<3.8051 -6.8724 -0.2407>, rotation=<rot 0 89.0417 -89.6665>}]
 Wing_Flap_l [{shapeType=Capsule, length=47.8196, radius=19.6853, center=<-4.7476 9.9036 -0.0547>, rotation=<rot 0 -99.0108 -89.8482>}]
 Wing_Flap_r [{shapeType=Capsule, length=47.8196, radius=19.6853, center=<4.7475 -9.9037 0.0547>, rotation=<rot 0 -99.0108 -89.8482>}]
}
constraints [
 {bone1Name=Engine_Base_r, bone2Name=Body_Offset_r, swing1LimitDegrees=45, swing2LimitDegrees=45, twistLimitDegrees=45}
 {bone1Name=Wing_Base_r, bone2Name=Body_Offset_r, swing1LimitDegrees=45, swing2LimitDegrees=45, twistLimitDegrees=45}
 {bone1Name=Front_Flap_r, bone2Name=Wing_Base_r, swing1LimitDegrees=45, swing2LimitDegrees=45, twistLimitDegrees=45}
 {bone1Name=Crossbar_r, bone2Name=Wing_Base_r, swing1LimitDegrees=45, swing2LimitDegrees=45, twistLimitDegrees=45}
 {bone1Name=Wing_Flap_Hinge_r, bone2Name=Wing_Base_r, swing1LimitDegrees=45, swing2LimitDegrees=45, twistLimitDegrees=45}
 {bone1Name=Wing_Flap_r, bone2Name=Wing_Flap_Hinge_r, swing1LimitDegrees=45, swing2LimitDegrees=45, twistLimitDegrees=45}
 {bone1Name=Parachute_Front_r, bone2Name=Parachute, swing1LimitDegrees=45, swing2LimitDegrees=45, twistLimitDegrees=45}
 {bone1Name=Parachute_Back_r, bone2Name=Parachute, swing1LimitDegrees=45, swing2LimitDegrees=45, twistLimitDegrees=45}
 {bone1Name=Parachute_Back_l, bone2Name=Parachute, swing1LimitDegrees=45, swing2LimitDegrees=45, twistLimitDegrees=45}
 {bone1Name=Parachute_Front_l, bone2Name=Parachute, swing1LimitDegrees=45, swing2LimitDegrees=45, twistLimitDegrees=45}
 {bone1Name=Front_Flap_l, bone2Name=Wing_Base_l, swing1LimitDegrees=45, swing2LimitDegrees=45, twistLimitDegrees=45}
 {bone1Name=Crossbar_l, bone2Name=Wing_Base_l, swing1LimitDegrees=45, swing2LimitDegrees=45, twistLimitDegrees=45}
 {bone1Name=Wing_Flap_Hinge_l, bone2Name=Wing_Base_l, swing1LimitDegrees=45, swing2LimitDegrees=45, twistLimitDegrees=45}
 {bone1Name=Wing_Flap_l, bone2Name=Wing_Flap_Hinge_l, swing1LimitDegrees=45, swing2LimitDegrees=45, twistLimitDegrees=45}
]
